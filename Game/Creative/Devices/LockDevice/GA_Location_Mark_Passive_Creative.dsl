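; Passive Creative lock-device ability that points the player toward the nearest relevant registered lock with a device-colored arrow.
; A server timer searches the item-requirement register and RPCs target changes to the owner, caching location for when the lock is out of relevancy.
; The client spawns the arrow emitter on the pawn and uses a zero-delay tick to rotate it, hiding it when the player is close and facing the lock.
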(event Ability|EventFailedToActivatePassiveAbility
  (params (ActorInfo :type Gameplay%20Ability%20Actor%20Info%20Structure)))

(event Ability|EventActivateAbility
  (CallFunction|InitAbility)
  (Ability|HasAuthority
    (:True
      (Utilities|Time|SetTimerbyEvent (delegate Custom|Server%20Find%20Best%20Actor%20Location) (Variables|Default|GetTimeBetweenLines) true) ; Server Query best target loop.
    )))

(event Custom|Server%20Find%20Best%20Actor%20Location
  (bind _details_changed (CallFunction|UpdateClosestLockDetails))
  (if _details_changed
    (CallFunction|RPCOWNERDetailsChanged (Variables|Default|GetLockActor) (Utilities|IsValid (Variables|Default|GetLockActor) :!class /Script/Engine.KismetSystemLibrary) (CallFunction|GetLockLocation) (Variables|Default|GetColor)) ; Server generates locations and information for the client to point towards\nas the client may not have a direct actor reference due to being out of relevancy range.
  ))

(event Custom|RPC_OWNER%20Details%20Changed
  (params (LockActor :type Actor%20Object%20Reference) (ValidLocation :type Boolean) (Actor%20Location :type Vector) (Color :type Linear%20Color%20Structure))
  (Variables|Default|SetLockActor LockActor) ; If we go out of relevancy this will be invalid.\nBut keeping a reference anyway for better fidelity \non a moving object when we're within relevancy.
  (Variables|Default|SetClientHasTarget ValidLocation)
  (Variables|Default|SetColor Color)
  (Variables|Default|SetCached%20Actor%20Location Actor%20Location)
  ;[1 0.803 0 0.1] Spawn a directional arrow to the target periodically on the client.
  (if (Variables|Default|GetClientHasTarget)
    (CallFunction|SpawnEmitter)
    (Utilities|FlowControl|RetriggerableDelay (Variables|Default|GetTimeBetweenLines))))

(event Custom|Client%20Tick%20Update%20Emitter
  ;[1 0.6509 0 0.25] TICKING EMITTER ROTATION - Don't like that we have to tick, but it's a necessity here to rotate smoothly.
  (Utilities|FlowControl|Delay 0.0)
  (if (Variables|Default|GetClientHasTarget)
    (Variables|Default|GetEmitter
      (:then
        (bind _returnvalue (CallFunction|ShouldShow))
        (Rendering|SetVisibility (Variables|Default|GetEmitter) _returnvalue false)
        (bind (_valid _returnvalue_2) (CallFunction|GetDesiredRotation))
        (if _valid
          (Transformation|SetWorldRotation (Variables|Default|GetEmitter) _returnvalue_2 false false)))
      (:else))))

(fn GetDesiredRotation
  (bind _outfortplayerpawn (CallFunction|GetAvatarasValidFortPlayerPawn)
    (:ValidFortPlayerPawn
      (return true (Math|Rotator|MakeRotfromX (* (- (Transformation|GetActorLocation _outfortplayerpawn) (CallFunction|GetLockLocation)) -1.0))))
    (:AvatarCastFailed
      (return false "0, 0, 0"))))

(fn GetLockLocation
  (bind _lockactor (Variables|Default|GetLockActor))
  (Utilities|IsValid _lockactor
    (:Is%20Valid
      (return (Transformation|GetActorLocation _lockactor)))
    (:Is%20Not%20Valid
      (return (Variables|Default|GetCached%20Actor%20Location)))))

(fn InitAbility
  (bind _outfortplayerpawn (CallFunction|GetAvatarasValidFortPlayerPawn)
    (:ValidFortPlayerPawn
      (bind _asfort_player_controller_athena (Utilities|Casting|CastToFortPlayerControllerAthena (Fortnite|GetFortPlayerControllerfromActor _outfortplayerpawn)))
      (Variables|Default|SetPlayerController _asfort_player_controller_athena))))

(fn SetNewLockActor
  (params (LockActor :type Actor%20Object%20Reference))
  (bind _lockactor (Variables|Default|GetLockActor))
  (bind _returnvalue (Transformation|GetActorLocation _lockactor))
  (if (!= LockActor (Variables|Default|GetLockActor))
    (Variables|Default|SetLockActor LockActor)
    (bind _lockactor_2 (Variables|Default|GetLockActor)
      (:then
        (Variables|Default|SetCached%20Actor%20Location _returnvalue)
        (bind _color (CreativeDevice|GetDeviceColor(Message) _lockactor_2))
        (Variables|Default|SetColor _color)
        (return true))
      (:else
        (return true)))
    (else
      (Utilities|IsValid LockActor
        (:Is%20Valid
          (if (Utilities|Operators|NotEqual(!=) _returnvalue (Variables|Default|GetCached%20Actor%20Location) 0.01)
            (Variables|Default|SetCached%20Actor%20Location _returnvalue)
            (return true)
            (else
              _)))
        (:Is%20Not%20Valid
          _))
      (return false))))

(fn ShouldShow
  ;[1 1 1] You must be either too far away or be looking in the wrong direction for the arrow to show. If you're close and looking at it, we want it to hide (as you'll be focusing on the hologram and we don't want to distract you)
  (bind _outfortplayerpawn (CallFunction|GetAvatarasValidFortPlayerPawn)
    (:ValidFortPlayerPawn
      (return (or (< (Math|Vector|DotProduct (Math|Vector|GetForwardVector (Camera|GetCameraRotation (Variables|/Script/Engine.PlayerController|GetPlayerCameraManager (Variables|Default|GetPlayerController)))) (Math|Vector|Normalize (- (CallFunction|GetLockLocation) (Transformation|GetActorLocation _outfortplayerpawn)))) (Variables|Default|GetMax%20Dot%20for%20Arrow)) (> (Math|Vector|VectorLength (- (CallFunction|GetLockLocation) (Transformation|GetActorLocation _outfortplayerpawn))) (Variables|Default|GetMin%20Distance%20for%20Arrow)))))
    (:AvatarCastFailed
      (return false))))

(fn SpawnEmitter
  (bind (_valid _returnvalue) (CallFunction|GetDesiredRotation))
  (bind _outfortplayerpawn (CallFunction|GetAvatarasValidFortPlayerPawn)
    (:ValidFortPlayerPawn
      (bind _returnvalue_2 (Effects|SpawnEmitterAttached "/Game/Athena/Items/EnvironmentalItems/Locks/Keycard/FX/P_Athena_KeycardLock_Direction_Creative.P_Athena_KeycardLock_Direction_Creative" (Variables|/Script/Engine.Character|GetMesh _outfortplayerpawn) "root" "0, 0,125.000000" _returnvalue "1.000000,1.000000,1.000000" "SnapToTarget"))
      (Utilities|IsValid _returnvalue_2
        (:Is%20Valid
          (Variables|/Script/Engine.SceneComponent|SetbAbsoluteRotation true _returnvalue_2)
          (bind _output_get (Variables|Default|SetEmitter _returnvalue_2))
          (Effects|Components|ParticleSystem|SetColorParameter _output_get "Color" (Variables|Default|GetColor))
          (CallFunction|ClientTickUpdateEmitter))))))

(fn UpdateClosestLockDetails
  (bind _returnvalue (Ability|GetCurrentSourceItemDefinition))
  (bind _pawn (LocalVariables|GetPawn))
  (bind _outfortplayerpawn (CallFunction|GetAvatarasValidFortPlayerPawn)
    (:ValidFortPlayerPawn
      (LocalVariables|SetPawn _outfortplayerpawn)
      (CreativeObjectRegistry|GetOrCreateObjectRegister _outfortplayerpawn "(TagName=\"Creative.Register.ItemSources.Requires\")" false 0 (Variables|Default|GetRegister%20Ref))
      (bind (_bisvalid _register) (Class|FortCreativeIndexedObjectRegisterComponent|GetRegisteratIndex (Variables|Default|GetRegister%20Ref) 0))
      (bind _register_objects (Utilities|Struct|BreakFortObjectArray _register))
      (for _array_element _register_objects
        (LocalVariables|SetCheckElem _array_element)
        (Utilities|IsValid _returnvalue
          (:Is%20Valid
            (bind _true (CreativeDevice|Items|IsHoldingItemforPawn(Message) (LocalVariables|GetCheckElem) (Creative|ItemVariant|ToItemVariantHandle(FortItemDefinition) (WeaponItemDefinition|GetAmmoWorldItemDefinition _returnvalue)) (LocalVariables|GetPawn)))
            (if _true
              _))
          (:Is%20Not%20Valid
            (bind _true_2 (CreativeDevice|Items|IsHoldingItemforPawn(Message) :self (LocalVariables|GetCheckElem) :Pawn (LocalVariables|GetPawn)))
            (if _true_2
              _)))
        (bind _asactor (Utilities|Casting|CastToActor (LocalVariables|GetCheckElem)))
        (if (< (Default|GetDistanceSquaredBetweenActors _pawn _asactor) (LocalVariables|GetBest%20Distance%20Squared))
          (LocalVariables|SetBest%20Distance%20Squared (Default|GetDistanceSquaredBetweenActors _pawn _asactor))
          (LocalVariables|SetBest%20Lock%20Actor _asactor)))
      (bind _details_changed (CallFunction|SetNewLockActor (LocalVariables|GetBest%20Lock%20Actor)))
      (return _details_changed))))
