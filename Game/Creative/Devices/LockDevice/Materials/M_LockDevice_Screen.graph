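# Lit material for the Creative Lock device's screen, showing locked or unlocked state as an emissive icon on a glowing, black-based surface.
# The State custom primitive value blends the background from red to blue and crossfades between two centre-scaled icon masks.
# A radial gradient, boosted by a tiled grid texture, shapes the colored glow.
# A PlacementTime timestamp transition modulates the icon only, before the result passes through MF_GameWideShared.
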Material /Game/Creative/Devices/LockDevice/Materials/M_LockDevice_Screen
graph {
 properties {bUseMaterialAttributes=true, pixelDepthOffsetMode=PDOM_Legacy, usedShadingModels=Default Lit}
 nodes [
  n1 = Constant()
  n2 = Power(Base=MaterialFunctionCall(
    TimestampA=@n1
    TransitionLength_In=@n1
    TimestampB=ScalarParameter() {bUseCustomPrimitiveData=true, defaultValue=1, parameterName=PlacementTime}
    TransitionLength_Out=Constant() {r=1}
    Transition In Only?=StaticBool()
   ) {materialFunction=&MF_Timestamp_InOut_01}.Linear)
  Add(A=@n2)
  n3 = Constant() {r=0.5}
  n4 = ScalarParameter() {bUseCustomPrimitiveData=true, defaultValue=1, parameterName=State, primitiveDataIndex=1}
  n5 = Constant() {r=0.65}
  n6 = Multiply(
   A=@n4
   B=TextureSample(UVs=MaterialFunctionCall(Texture Scale=@n5) {materialFunction=&ScaleUVsByCenter}.UVs) {samplerType=SAMPLERTYPE_LinearColor, texture=/Game/Creative/Devices/Lock_and_Key/Textures/T_LockKeyRequirements_M}.B
  )
  n7 = Constant() {r=0.8}
  n8 = Multiply(
   A=OneMinus(None=@n4)
   B=TextureSample(UVs=MaterialFunctionCall(Texture Scale=@n7) {materialFunction=&ScaleUVsByCenter}.UVs) {samplerType=SAMPLERTYPE_LinearColor, texture=/Game/Creative/Devices/Lock_and_Key/Textures/T_LockHiddenKey}.R
  )
  n9 = Multiply(A=Multiply(A=@n2, B=Add(A=@n6, B=@n8))) {constB=5}
  n10 = LinearInterpolate(
   A=VectorParameter() {parameterName=Locked State Color, defaultValue=<rgba 1 0 0 1>}.RGB
   B=VectorParameter() {parameterName=Unlocked State Color, defaultValue=<rgba 0.0143 0.1053 1 1>}.RGB
   Alpha=@n4
  )
  n11 = Multiply(A=@n10, B=ScalarParameter() {defaultValue=3, parameterName=Emissive Strength})
  n12 = MaterialFunctionCall() {materialFunction=&RadialGradientExponential}
  n13 = TextureCoordinate()
  n14 = ScalarParameter() {defaultValue=5, parameterName=Grid Tiling}
  n15 = Multiply(
   A=TextureSample(UVs=Multiply(A=@n13, B=@n14)) {texture=/Game/Effects/Fort_Effects/Effects/Characters/Generic/ResIn/DiffusedStylizedGrid}.R
   B=ScalarParameter() {defaultValue=1, parameterName=Grid Strength}
  )
  n16 = Multiply(A=@n11, B=Add(A=@n12.RadialGradientExponential, B=Multiply(A=@n12.RadialGradientExponential, B=@n15)))
  MaterialFunctionCall!(In=MakeMaterialAttributes!(
    BaseColor=Constant()
    Metallic
    Specular=@n3
    Roughness=@n3
    Anisotropy
    EmissiveColor=Add(A=@n9, B=@n16)
    Opacity
    OpacityMask
    Normal
    Tangent
    WorldPositionOffset
    SubsurfaceColor
    ClearCoat
    ClearCoatRoughness
    AmbientOcclusion
    Refraction
    CustomizedUVs_0
    CustomizedUVs_1
    CustomizedUVs_2
    CustomizedUVs_3
    CustomizedUVs_4
    CustomizedUVs_5
    CustomizedUVs_6
    CustomizedUVs_7
    PixelDepthOffset
    ShadingModel
    Displacement
   )) {materialFunction=/Game/Athena/WinterArt/Materials/MF_GameWideShared}
  Multiply(A=@n11, B=@n10)
 ]
}
